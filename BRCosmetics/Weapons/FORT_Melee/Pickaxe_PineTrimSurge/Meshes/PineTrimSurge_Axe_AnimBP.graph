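# Anim Blueprint for the PineTrimSurge pickaxe that gives its three-bone rope spring-driven secondary motion under gravity.
# An additive rotation of deform_rope_MSTR feeds one of three AnimDynamics chains: root-relative or world in the front end, depending on bIsOnPawn, and bone-relative otherwise.
# The event graph sets the front-end rotator, plus VelScale and a deprecated-function gravity override, neither visibly read by the AnimGraph.
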AnimBlueprint /BRCosmetics/Weapons/FORT_Melee/Pickaxe_PineTrimSurge/Meshes/PineTrimSurge_Axe_AnimBP
target_skeleton=&/PineTrimSurge_Axe_Skeleton
graphs [
 AnimGraph {nodes [
   n1 = Anim Dynamics - Bone: dyn_rope_1(ComponentPose=Local To Component(LocalPose=Use cached pose 'IsPawnRotated'()), Alpha=1, ExternalForce="0,0,0") {
    angularDampingOverride=1
    angularSpringConstant=500
    bAngularSpring=true
    bOverrideAngularDamping=true
    bUseGravityOverride=true
    simulationSpace=RootRelative
    alphaBoolBlend {blendOption=Linear}
    boundBone {boneName=dyn_rope_1}
    componentAppliedLinearAccClamp=<2000 2000 2000>
    componentLinearAccScale=<0.2 0.2 0.2>
    componentLinearVelScale=<0.2 0.2 0.2>
    gravityOverride {z=-980}
    physicsBodyDefinitions [{
      boundBone {boneName=dyn_rope_1}
      boxExtents=<12 12 12>
      constraintSetup {angularLimitsMax=<5 40 35>, angularLimitsMin=<-5 -40 -45>, angularTarget=<1 0 0>}
      localJointOffset=<1.5 0 0>
     }]
    relativeSpaceBone {boneName=b_AR_Root}
   }
   n2 = Anim Dynamics - Bone: dyn_rope_2(ComponentPose=@n1, Alpha=1, ExternalForce="0,0,0") {
    angularDampingOverride=1
    angularSpringConstant=500
    bAngularSpring=true
    bOverrideAngularDamping=true
    bUseGravityOverride=true
    simulationSpace=RootRelative
    alphaBoolBlend {blendOption=Linear}
    boundBone {boneName=dyn_rope_2}
    componentAppliedLinearAccClamp=<2000 2000 2000>
    componentLinearAccScale=<0.2 0.2 0.2>
    componentLinearVelScale=<0.2 0.2 0.2>
    gravityOverride {z=-980}
    physicsBodyDefinitions [{
      boundBone {boneName=dyn_rope_2}
      boxExtents=<12 12 12>
      constraintSetup {angularLimitsMax=<10 40 40>, angularLimitsMin=<-10 -40 -40>, angularTarget=<1 0 0>}
      localJointOffset=<1.5 0 0>
     }]
    relativeSpaceBone {boneName=b_AR_Root}
   }
   n3 = Anim Dynamics - Bone: dyn_rope_3(ComponentPose=@n2, Alpha=1, ExternalForce="0,0,0") {
    angularDampingOverride=1
    angularSpringConstant=500
    bAngularSpring=true
    bOverrideAngularDamping=true
    bUseGravityOverride=true
    simulationSpace=RootRelative
    alphaBoolBlend {blendOption=Linear}
    boundBone {boneName=dyn_rope_3}
    componentAppliedLinearAccClamp=<2000 2000 2000>
    componentLinearAccScale=<0.2 0.2 0.2>
    componentLinearVelScale=<0.2 0.2 0.2>
    gravityOverride {z=-980}
    physicsBodyDefinitions [{
      boundBone {boneName=dyn_rope_3}
      boxExtents=<12 12 12>
      constraintSetup {angularLimitsMax=<10 40 60>, angularLimitsMin=<-10 -40 -60>, angularTarget=<1 0 0>}
      localJointOffset=<1.5 0 0>
     }]
    relativeSpaceBone {boneName=b_AR_Root}
    simSpaceSettings {maxAngularAcceleration=3000, maxAngularVelocity=3000, simSpaceAngularAlpha=0.5}
   }
   n4 = Component To Local(ComponentPose=@n3)
   n5 = Anim Dynamics - Bone: dyn_rope_1(ComponentPose=Local To Component(LocalPose=Use cached pose 'IsPawnRotated'()), Alpha=1, ExternalForce="0,0,0") {
    angularDampingOverride=1
    angularSpringConstant=500
    bAngularSpring=true
    bOverrideAngularDamping=true
    bUseGravityOverride=true
    simulationSpace=World
    alphaBoolBlend {blendOption=Linear}
    boundBone {boneName=dyn_rope_1}
    componentAppliedLinearAccClamp=<2000 2000 2000>
    componentLinearAccScale=<0.2 0.2 0.2>
    componentLinearVelScale=<0.2 0.2 0.2>
    gravityOverride {z=-980}
    physicsBodyDefinitions [{
      boundBone {boneName=dyn_rope_1}
      boxExtents=<11 11 11>
      constraintSetup {angularLimitsMax=<5 80 35>, angularLimitsMin=<-5 -5 -45>, angularTarget=<1 0 0>}
      localJointOffset=<1.1 0 0>
     }]
    relativeSpaceBone {boneName=b_AR_Root}
   }
   n6 = Anim Dynamics - Bone: dyn_rope_2(ComponentPose=@n5, Alpha=1, ExternalForce="0,0,0") {
    angularDampingOverride=1
    angularSpringConstant=500
    bAngularSpring=true
    bOverrideAngularDamping=true
    bUseGravityOverride=true
    simulationSpace=World
    alphaBoolBlend {blendOption=Linear}
    boundBone {boneName=dyn_rope_2}
    componentAppliedLinearAccClamp=<2000 2000 2000>
    componentLinearAccScale=<0.2 0.2 0.2>
    componentLinearVelScale=<0.2 0.2 0.2>
    gravityOverride {z=-980}
    physicsBodyDefinitions [{
      boundBone {boneName=dyn_rope_2}
      boxExtents=<11 11 11>
      constraintSetup {angularLimitsMax=<10 40 40>, angularLimitsMin=<-10 -40 -40>, angularTarget=<1 0 0>}
      localJointOffset=<1.1 0 0>
     }]
    relativeSpaceBone {boneName=b_AR_Root}
   }
   n7 = Anim Dynamics - Bone: dyn_rope_3(ComponentPose=@n6, Alpha=1, ExternalForce="0,0,0") {
    angularDampingOverride=1
    angularSpringConstant=500
    bAngularSpring=true
    bOverrideAngularDamping=true
    bUseGravityOverride=true
    simulationSpace=World
    alphaBoolBlend {blendOption=Linear}
    boundBone {boneName=dyn_rope_3}
    componentAppliedLinearAccClamp=<2000 2000 2000>
    componentLinearAccScale=<0.2 0.2 0.2>
    componentLinearVelScale=<0.2 0.2 0.2>
    gravityOverride {z=-980}
    physicsBodyDefinitions [{
      boundBone {boneName=dyn_rope_3}
      boxExtents=<11 11 11>
      constraintSetup {angularLimitsMax=<10 40 40>, angularLimitsMin=<-10 -40 -40>, angularTarget=<1 0 0>}
      localJointOffset=<1.1 0 0>
     }]
    relativeSpaceBone {boneName=b_AR_Root}
   }
   n8 = Component To Local(ComponentPose=@n7)
   n9 = Get bIsOnPawn() -> (bIsOnPawn:bool)
   n10 = "Transform (Modify) Bone - Bone: deform_rope_MSTR"(
    ComponentPose=Local To Component(LocalPose=Use cached pose 'InputPoseCache'())
    Alpha=1
    Translation="0,0,0"
    Rotation="0,-90,80"
    Scale="1,1,1"
   ) {rotationMode=BMM_Additive, alphaBoolBlend {blendOption=Linear}, boneToModify {boneName=deform_rope_MSTR}, rotation {roll=80, yaw=-90}}
   n11 = Anim Dynamics - Bone: dyn_rope_1(ComponentPose=@n10, Alpha=1, ExternalForce="0,0,0") {
    angularDampingOverride=1
    angularSpringConstant=500
    bAngularSpring=true
    bOverrideAngularDamping=true
    bUseGravityOverride=true
    simulationSpace=BoneRelative
    alphaBoolBlend {blendOption=Linear}
    boundBone {boneName=dyn_rope_1}
    componentAppliedLinearAccClamp=<2000 2000 2000>
    componentLinearAccScale=<0.4 0.4 0.4>
    componentLinearVelScale=<0.4 0.4 0.4>
    gravityOverride {z=-980}
    physicsBodyDefinitions [{
      boundBone {boneName=dyn_rope_1}
      boxExtents=<10 10 10>
      constraintSetup {angularLimitsMax=<5 80 35>, angularLimitsMin=<-5 -5 -45>, angularTarget=<1 0 0>}
      localJointOffset=<1.7 0 0>
     }]
    relativeSpaceBone {boneName=b_AR_Root}
   }
   n12 = Anim Dynamics - Bone: dyn_rope_2(ComponentPose=@n11, Alpha=1, ExternalForce="0,0,0") {
    angularDampingOverride=1
    angularSpringConstant=500
    bAngularSpring=true
    bOverrideAngularDamping=true
    bUseGravityOverride=true
    simulationSpace=BoneRelative
    alphaBoolBlend {blendOption=Linear}
    boundBone {boneName=dyn_rope_2}
    componentAppliedLinearAccClamp=<2000 2000 2000>
    componentLinearAccScale=<0.4 0.4 0.4>
    componentLinearVelScale=<0.4 0.4 0.4>
    gravityOverride {z=-980}
    physicsBodyDefinitions [{
      boundBone {boneName=dyn_rope_2}
      boxExtents=<10 10 10>
      constraintSetup {angularLimitsMax=<10 40 40>, angularLimitsMin=<-10 -40 -40>, angularTarget=<1 0 0>}
      localJointOffset=<1.7 0 0>
     }]
    relativeSpaceBone {boneName=b_AR_Root}
   }
   n13 = Anim Dynamics - Bone: dyn_rope_3(ComponentPose=@n12, Alpha=1, ExternalForce="0,0,0") {
    angularDampingOverride=1
    angularSpringConstant=500
    bAngularSpring=true
    bOverrideAngularDamping=true
    bUseGravityOverride=true
    simulationSpace=BoneRelative
    alphaBoolBlend {blendOption=Linear}
    boundBone {boneName=dyn_rope_3}
    componentAppliedLinearAccClamp=<2000 2000 2000>
    componentLinearAccScale=<0.4 0.4 0.4>
    componentLinearVelScale=<0.4 0.4 0.4>
    gravityOverride {z=-980}
    physicsBodyDefinitions [{
      boundBone {boneName=dyn_rope_3}
      boxExtents=<10 10 10>
      constraintSetup {angularLimitsMax=<10 40 60>, angularLimitsMin=<-10 -40 -60>, angularTarget=<1 0 0>}
      localJointOffset=<1.7 0 0>
     }]
    relativeSpaceBone {boneName=b_AR_Root}
    simSpaceSettings {maxAngularAcceleration=3000, maxAngularVelocity=3000, simSpaceAngularAlpha=0.5}
   }
   Output Pose(Result=Blend Poses by bool(
     BlendPose_0=Blend Poses by bool(BlendPose_0=@n4, BlendPose_1=@n8, bActiveValue=@n9) {blendType=Linear}
     BlendPose_1=Component To Local(ComponentPose=@n13)
     bActiveValue=Get bIsInFrontEnd()
    ) {blendType=Linear}) {layerGroup=None}
   AnimGraphNode_SubInput_0 = Input Pose()
   Save cached pose 'InputPoseCache'(Pose=@AnimGraphNode_SubInput_0) {cacheName=InputPoseCache}
   n14 = Local To Component(LocalPose=Use cached pose 'InputPoseCache'())
   n15 = Get PickAxeFrontEndRotator() -> (PickAxeFrontEndRotator:Rotator)
   Save cached pose 'IsPawnRotated'(Pose=Component To Local(ComponentPose="Transform (Modify) Bone - Bone: deform_rope_MSTR"(ComponentPose=@n14, Alpha=1, Translation="0,0,0", Rotation=@n15, Scale="1,1,1") {rotationMode=BMM_Additive, alphaBoolBlend {blendOption=Linear}, boneToModify {boneName=deform_rope_MSTR}})) {cacheName=IsPawnRotated}
  ]}
 EventGraph {nodes [
   Event BlueprintUpdateAnimation()
   > n1 = Sequence()
   n2 = Branch(execute=@n1.then_1, Condition=OR Boolean(A=Get bIsFalling(), B=Get bIsJumping()))
   > Set VelScale(VelScale:Vector="0.4,0.4,0.4") -> (Output_Get:Vector)
   > Branch(Condition=Get bIsFalling())
   > Set VelScale(VelScale="0.35,0.35,0.35")
   Set VelScale(execute=@n2.else, VelScale="0.45,0.45,0.45")
   > Branch(Condition=Get bIsInFrontEnd())
   > Set VelScale(VelScale="0.1,0.25,0.25")
   > Set PickAxeFrontEndRotator(PickAxeFrontEndRotator:Rotator=SelectRotator!(A:Rotator="0,-90,80", B:Rotator="0,-90,30", bPickA:bool=Get bIsOnPawn()) -> (ReturnValue:Rotator)) -> (Output_Get:Rotator)
   n3 = Get Gravity Override Params()
   n4 = Gravity Override State Function DEPRECATED(execute=@n1.then_0, CharacterPartAnimInstance=Self-Reference(), GameTimeInSeconds=GetGameTimeInSeconds(), GravityOverrideParams=@n3)
   > Set Gravity Override(Gravity Override=@n4)
   Set Gravity Override Params(Gravity Override Params:Gravity Override Params Struct Structure=@n3) -> (Output_Get:Gravity Override Params Struct Structure)
  ]}
]
variables [
 Gravity Override {type=Vector}
 Gravity Override Params {type=GravityOverrideParamsStruct}
 PickAxeFrontEndRotator {type=Rotator}
 VelScale {type=Vector}
]
