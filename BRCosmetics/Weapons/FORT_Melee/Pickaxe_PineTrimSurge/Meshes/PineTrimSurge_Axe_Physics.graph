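# Physics asset for the PineTrimSurge pickaxe, defining collision and joints for a short dangling rope on the two-handed axe.
# A large capsule covers the handle, a tiny capsule anchors the rope master bone, and three identical capsules form the rope segments.
# Constraints chain handle, master and segments, with free rotation at the handle joint, a wide swing limit on the first segment, tighter swing on the next two, and little twist along the rope.
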PhysicsAsset /BRCosmetics/Weapons/FORT_Melee/Pickaxe_PineTrimSurge/Meshes/PineTrimSurge_Axe_Physics
bodies {
 Melee_TwoHanded_Handle [{shapeType=Capsule, length=34.6113, radius=7.5106, center=<-0.0061 0.0326 12.0706>, rotation=<rot 0.0793 -90 -0.0147>}]
 deform_rope_MSTR [{shapeType=Capsule, length=0.505, radius=0.505, shapeName=Base_sword_chain_capsule, rotation=<rot 0 -90 0>}]
 dyn_rope_1 [{
   shapeType=Capsule
   length=1.821
   radius=1.7
   shapeName=dyn_sword_chain_02_capsule
   center=<1.4239 0.0032 -0.0669>
   rotation=<rot 0 -86.7509 -88.8639>
  }]
 dyn_rope_2 [{
   shapeType=Capsule
   length=1.821
   radius=1.7
   shapeName=dyn_sword_chain_02_capsule
   center=<1.4239 0.0032 -0.0669>
   rotation=<rot 0 -86.7509 -88.8639>
  }]
 dyn_rope_3 [{
   shapeType=Capsule
   length=1.821
   radius=1.7
   shapeName=dyn_sword_chain_02_capsule
   center=<1.4239 0.0032 -0.0669>
   rotation=<rot 0 -86.7509 -88.8639>
  }]
}
constraints [
 {bone1Name=dyn_rope_1, bone2Name=deform_rope_MSTR, swing1LimitDegrees=80, swing2LimitDegrees=80, twistLimitDegrees=5}
 {bone1Name=dyn_rope_2, bone2Name=dyn_rope_1, swing1LimitDegrees=40, swing2LimitDegrees=40, twistLimitDegrees=5}
 {bone1Name=dyn_rope_3, bone2Name=dyn_rope_2, swing1LimitDegrees=40, swing2LimitDegrees=40, twistLimitDegrees=5}
 {
  bone1Name=deform_rope_MSTR
  bone2Name=Melee_TwoHanded_Handle
  swing1LimitDegrees=45
  swing1Motion=Free
  swing2LimitDegrees=45
  swing2Motion=Free
  twistLimitDegrees=45
  twistMotion=Free
 }
]
